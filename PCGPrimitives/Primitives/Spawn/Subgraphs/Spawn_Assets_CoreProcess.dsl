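; PCG core subgraph that spawns static meshes, PCG Data Assets and Blueprint actors on input points, or on points combined from splines and polygons.
; Seeds offset per data set drive random or index-based asset selection, unless a mesh attribute supplies it.
; Optional bounds-matched scaling, offsets and random transforms follow.
; Points route by asset class to guarded spawners, with data assets expanded into mesh points and OverlayMaterial applied as a mesh material override.
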(graph Spawn_Assets_CoreProcess
  (params (Assets :type (Array (SoftObject /Script/CoreUObject.Object))) (PositionOffset :type Vector) (RotationOffset :type Rotator) (ScaleOffset :type Vector :default (1.0 1.0 1.0)) (Match_scale_to_input_bounds :type Bool) (UniformScale :type Bool :default true) (Use_Mesh_Attribute :type Bool) (Mesh_Attrinute :type String) (Use_Index_attribute :type Bool) (Index_attribute :type String) (Random_Position_Range :type Vector) (Random_Rotation_Range :type Vector) (Random_Scale_Range :type Vector) (seed :type Int) (OverlayMaterial :type (SoftObject /Script/Engine.MaterialInterface)))
  (bind MergePoints_1 (Merge%20Points))
  (bind Branch_1 (Branch :bOutputToB-param Use_Index_attribute)) ; Two asset selection modes:\n- Output A: seed-based random assignment\n- Output B: index attribute driven\n  (deterministic, user-controlled)\nControlled by Use Index Attribute.
  (bind AttributeRename_2 (Attribute%20Rename :AttributeToRename Assets :NewAttributeName Asset :In Assets))
  (bind CopyAttributes_1 (Copy%20Attributes :InputSource $Index :OutputTarget ind))
  (bind MatchAndSetAttributes_1 (Match%20And%20Set%20Attributes :bMatchAttributes true :InputAttribute $Seed :MatchAttribute ind :InputAttribute-param Index_attribute))
  (bind MatchAndSetAttributes_0 (Match%20And%20Set%20Attributes :InputAttribute $Seed :MatchAttribute ind))
  (bind AttributeRename_0 (Attribute%20Rename :AttributeToRename Mesh :NewAttributeName Asset :AttributeToRename-param Mesh_Attrinute))
  (bind Select_0 (Select :bUseInputB-param Use_Mesh_Attribute))
  (bind Branch_0 (Branch :bOutputToB-param Match_scale_to_input_bounds))
  (bind AttributeMathsOp_5 (Attribute%20Maths%20Op :InputSource1 $Scale :InputSource2 Scale :Operation Set))
  (bind FilterDataByType_4 (Filter%20Data%20By%20Type :TargetType (struct :Ids [(struct :Struct /Script/PCG.PCGDataTypeInfoPoint)])))
  (bind AttributeFilter_0 (Filter%20Attribute%20Elements :AttributeTypes (struct :Type SoftObjectPath) :bUseConstantThreshold true :Operator Matches :TargetAttribute Asset :ThresholdAttribute @Last)) ; fail safe if no asset provided, prevent warning on next node\nImportant because if thousands of point arewithout asset reference,\neach point will porduce warning and freeze the edito.
  (bind GetClassFromAttribute_0 (Get%20Class%20From%20Attribute :InputSource Asset :OutputTarget "Asset Type Str")) ; Detect asset type from the Asset attribute.\nPoints are routed to the correct spawner\n(StaticMesh, PCGDataAsset, Blueprint).\nDataNum guards each branch — skips the\nspawner if no matching assets exist,\navoiding empty spawner warnings.
  (bind CreateConstant_1 (Create%20Constant :AttributeTypes (struct :Type String)))
  (bind AttributeCompareOp_1 (Attribute%20Compare%20Op :Operation NotEqual :InA OverlayMaterial))
  (bind Branch_3 (Branch))
  (bind StaticMeshSpawner_1 (Static%20Mesh%20Spawner :InstanceDataPackerParameters (class:/Script/PCG.PCGInstanceDataPackerByAttribute :AttributeSelectors [$Density $Extents $Color]) :InstanceDataPackerType /Script/PCG.PCGInstanceDataPackerByAttribute :MeshSelectorParameters (class:/Script/PCG.PCGMeshSelectorByAttribute :AttributeName Asset :TemplateDescriptor (struct :bHasPerInstanceHitProxies true)) :MeshSelectorType /Script/PCG.PCGMeshSelectorByAttribute))
  (bind AddAttribute_1 (Add%20Attribute :OutputTarget OverrideMaterial :Attributes OverlayMaterial))
  (bind StaticMeshSpawner_2 (Static%20Mesh%20Spawner :InstanceDataPackerParameters (class:/Script/PCG.PCGInstanceDataPackerByAttribute :AttributeSelectors [$Density $Extents $Color]) :InstanceDataPackerType /Script/PCG.PCGInstanceDataPackerByAttribute :MeshSelectorParameters (class:/Script/PCG.PCGMeshSelectorByAttribute :AttributeName Asset :bUseAttributeMaterialOverrides true :MaterialOverrideAttributes [OverrideMaterial] :TemplateDescriptor (struct :bHasPerInstanceHitProxies true)) :MeshSelectorType /Script/PCG.PCGMeshSelectorByAttribute))
  (bind SpawnActor_1 (Spawn%20Actor :bSpawnByAttribute true :SpawnAttribute Asset))
  (bind Gather_0 (Gather))
  ;[0.15 0.15 0.15 0.5] DIFFERENT treatment for different types
  (bind FilterDataByType_2 (Filter%20Data%20By%20Type :TargetType (struct :Ids [(struct :Struct /Script/PCG.PCGDataTypeInfoPoint)])))
  ;;[0.15 0.15 0.15 0.5] treating spline and polygons as one entry
  (bind FilterDataByType_0 (Filter%20Data%20By%20Type :TargetType (struct :Ids [(struct :Struct /Script/PCG.PCGDataTypeInfoSpline)])))
  (bind FilterDataByType_1 (Filter%20Data%20By%20Type :TargetType (struct :Ids [(struct :Struct /Script/PCG.PCGDataTypeInfoPolygon2D)])))
  (bind GetSplineControlPoints_0 (Get%20Spline%20Control%20Points))
  (bind MergePoints_0 (Merge%20Points))
  (bind CombinePoints_1 (Combine%20Points))
  ;[0.15 0.15 0.15 0.5] seed per data stream
  (bind WriteDataIndex_1 (Write%20Data%20Index :bAddAttribute true :IndexAttribute dataIndex))
  (bind AttributeMathsOp_13 (Attribute%20Maths%20Op :InputSource1 $Seed :InputSource2 dataIndex)) ; Offset the seed by the data partition index\nso each collection of points gets a unique\nrandom seed, avoiding identical results\nacross multiple input data sets.
  ;[0.15 0.15 0.15 0.5] SOLVING scale, the mesh size will match the input point size
  (bind FilterDataByType_3 (Filter%20Data%20By%20Type :TargetType (struct :Ids [(struct :Struct /Script/PCG.PCGDataTypeInfoPoint)])))
  (bind BoundsFromMesh_0 (Bounds%20From%20Mesh :MeshAttribute Asset))
  (bind AttributeMathsOp_6 (Attribute%20Maths%20Op :InputSource1 $BoundsMax :InputSource2 $BoundsMin :Operation Subtract :OutputTarget MeshSize)) ; SIZE
  (bind FilterDataByType_5 (Filter%20Data%20By%20Type :TargetType (struct :Ids [(struct :Struct /Script/PCG.PCGDataTypeInfoPoint)])))
  (bind AttributeMathsOp_7 (Attribute%20Maths%20Op :InputSource1 $BoundsMax :InputSource2 $BoundsMin :Operation Subtract :OutputTarget Size))
  (bind CopyAttributes_0 (Copy%20Attributes :InputSource MeshSize :OutputTarget MeshSize))
  (bind Branch_2 (Branch :bOutputToB-param UniformScale))
  (bind AttributeMathsOp_4 (Attribute%20Maths%20Op :InputSource1 Size :InputSource2 MeshSize :Operation Divide :OutputTarget Scale))
  (bind AttributeMathsOp_14 (Attribute%20Maths%20Op :InputSource1 SizeMax :InputSource2 MeshSizeMax :Operation Divide :OutputTarget Scale))
  ;;[0.15 0.15 0.15 0.5] MAX input bounds size
  (bind ExtractAttribute_1 (Extract%20Attribute%20at%20Index :bForceOutputAttributeToBeInElementsDomain true :InputSource Size.X))
  (bind ExtractAttribute_2 (Extract%20Attribute%20at%20Index :bForceOutputAttributeToBeInElementsDomain true :InputSource Size.Y))
  (bind AttributeMathsOp_0 (Attribute%20Maths%20Op :Operation Max))
  (bind ExtractAttribute_3 (Extract%20Attribute%20at%20Index :bForceOutputAttributeToBeInElementsDomain true :InputSource Size.Z))
  (bind AttributeMathsOp_1 (Attribute%20Maths%20Op :Operation Max :OutputTarget SizeMax))
  ;[0.15 0.15 0.15 0.5] Max mesh reference Size
  (bind ExtractAttribute_4 (Extract%20Attribute%20at%20Index :bForceOutputAttributeToBeInElementsDomain true :InputSource MeshSize.X))
  (bind ExtractAttribute_5 (Extract%20Attribute%20at%20Index :bForceOutputAttributeToBeInElementsDomain true :InputSource MeshSize.Y))
  (bind AttributeMathsOp_2 (Attribute%20Maths%20Op :Operation Max))
  (bind ExtractAttribute_6 (Extract%20Attribute%20at%20Index :bForceOutputAttributeToBeInElementsDomain true :InputSource MeshSize.Z))
  (bind AttributeMathsOp_3 (Attribute%20Maths%20Op :Operation Max :OutputTarget MeshSizeMax))
  ;[0.15 0.15 0.15 0.5] OFFSETS
  (bind TransformPoints_1 (Transform%20Points :bUniformScale false :bUniformScale-param UniformScale :OffsetMax-param PositionOffset :OffsetMin-param PositionOffset :RotationMax-param RotationOffset :RotationMin-param RotationOffset :ScaleMax-param ScaleOffset :ScaleMin-param ScaleOffset))
  ;[0.15 0.15 0.15 0.5] RANDOMIZE TRANSFORM
  (bind AttributeMathsOp_8 (Attribute%20Maths%20Op :Operation Multiply :InA Random_Position_Range))
  (bind BreakVectorAttribute_0 (Break%20Vector%20Attribute :In Random_Rotation_Range))
  (bind MakeRotatorAttribute_0 (Make%20Rotator%20Attribute :Operation MakeRotFromAngles))
  (bind AttributeMathsOp_9 (Attribute%20Maths%20Op :Operation Multiply :InA Random_Rotation_Range))
  (bind BreakVectorAttribute_1 (Break%20Vector%20Attribute))
  (bind MakeRotatorAttribute_1 (Make%20Rotator%20Attribute :Operation MakeRotFromAngles))
  (bind AttributeMathsOp_10 (Attribute%20Maths%20Op :InA Random_Scale_Range))
  (bind CreateConstant_0 (Create%20Constant :AttributeTypes (struct :Type Vector :DoubleValue 1.0 :VectorValue (1.0 1.0 1.0))))
  (bind AttributeMathsOp_11 (Attribute%20Maths%20Op :Operation Subtract :InB Random_Scale_Range))
  (bind TransformPoints_2 (Transform%20Points :bUniformScale false :OffsetMax (-3e+03 0.0 0.0) :OffsetMin (3e+03 0.0 0.0) :bUniformScale-param UniformScale :OffsetMax-param Random_Position_Range :Seed-param seed))
  ;[0.15 0.15 0.15 0.5] StaticMesh
  (bind CreateConstant_3 (Create%20Constant :AttributeTypes (struct :Type String :StringValue "StaticMesh")))
  (bind AttributeFilter_2 (Filter%20Attribute%20Elements :bGenerateOutputDataEvenIfEmpty false :bWarnOnDataMissingAttribute false :Operator Matches :TargetAttribute "Asset Type Str" :ThresholdAttribute @Last))
  (bind DataNum_0 (Data%20Count))
  (bind Branch_8 (Branch))
  ;[0.15 0.15 0.15 0.5] PCG Data Assets
  (bind CreateConstant_4 (Create%20Constant :AttributeTypes (struct :Type String :StringValue "PCGdataAsset")))
  (bind AttributeFilter_3 (Filter%20Attribute%20Elements :bGenerateOutputDataEvenIfEmpty false :bWarnOnDataMissingAttribute false :Operator Matches :TargetAttribute "Asset Type Str" :ThresholdAttribute @Last))
  (bind DataNum_1 (Data%20Count))
  (bind Branch_9 (Branch))
  ;[0.15 0.15 0.15 0.5] ASSEMBLIES
  (bind AttributePartition_1 (Attribute%20Partition :PartitionAttributeSelectors [Asset]))
  (bind ExtractAttribute_7 (Extract%20Attribute%20at%20Index :bForceOutputAttributeToBeInElementsDomain true :InputSource Asset))
  (bind AttributeFilter_1 (Filter%20Attribute%20Elements :AttributeTypes (struct :Type String) :bUseConstantThreshold true :Operator Matches :TargetAttribute Asset :ThresholdAttribute @Last)) ; FAIL SAFE: points with an empty Asset string\nare routed out here. The OutsideFilter feeds\nthe PCGDataAsset load path — empty assets\nare dropped naturally downstream.
  (bind PCGLoadDataAsset_1 (Load%20PCG%20Data%20Asset :bLoadFromInput true))
  (bind ToPoint_1 (To%20Point))
  (bind FilterDataByTag_1 (Filter%20Data%20By%20Tag :SelectedTags "points"))
  (bind AttributeRename_1 (Attribute%20Rename :AttributeToRename Mesh :NewAttributeName Asset))
  (bind CopyPoints_1 (Copy%20Points :AttributeInheritance SourceOnly :bCopyEachSourceOnEveryTarget false :TagInheritance Source))
  ;[0.15 0.15 0.15 0.5] Blueprints
  (bind CreateConstant_2 (Create%20Constant :AttributeTypes (struct :Type String :StringValue "Blueprint")))
  (bind AttributeFilter_4 (Filter%20Attribute%20Elements :bGenerateOutputDataEvenIfEmpty false :bWarnOnDataMissingAttribute false :Operator Matches :TargetAttribute "Asset Type Str" :ThresholdAttribute @Last))
  (bind DataNum_2 (Data%20Count))
  (bind Branch_10 (Branch))
  (connect Branch_3 Output%20B AddAttribute_1 In)
  (connect CreateConstant_1 Out AttributeCompareOp_1 InB)
  (connect TransformPoints_2 Out AttributeFilter_0 In)
  (connect ExtractAttribute_7 Out AttributeFilter_1 In)
  (connect CreateConstant_3 Out AttributeFilter_2 Filter)
  (connect GetClassFromAttribute_0 Out AttributeFilter_2 In)
  (connect CreateConstant_4 Out AttributeFilter_3 Filter)
  (connect GetClassFromAttribute_0 Out AttributeFilter_3 In)
  (connect CreateConstant_2 Out AttributeFilter_4 Filter)
  (connect GetClassFromAttribute_0 Out AttributeFilter_4 In)
  (connect ExtractAttribute_1 Out AttributeMathsOp_0 InA)
  (connect ExtractAttribute_2 Out AttributeMathsOp_0 InB)
  (connect AttributeMathsOp_0 Out AttributeMathsOp_1 InA)
  (connect ExtractAttribute_3 Out AttributeMathsOp_1 InB)
  (connect ExtractAttribute_4 Out AttributeMathsOp_2 InA)
  (connect ExtractAttribute_5 Out AttributeMathsOp_2 InB)
  (connect AttributeMathsOp_2 Out AttributeMathsOp_3 InA)
  (connect ExtractAttribute_6 Out AttributeMathsOp_3 InB)
  (connect Branch_2 Output%20A AttributeMathsOp_4 InA)
  (connect Branch_2 Output%20A AttributeMathsOp_4 InB)
  (connect Branch_0 Output%20B AttributeMathsOp_5 InA)
  (connect AttributeMathsOp_4 Out AttributeMathsOp_5 InB)
  (connect AttributeMathsOp_14 Out AttributeMathsOp_5 InB)
  (connect BoundsFromMesh_0 Out AttributeMathsOp_6 InA)
  (connect BoundsFromMesh_0 Out AttributeMathsOp_6 InB)
  (connect FilterDataByType_5 InsideFilter AttributeMathsOp_7 InA)
  (connect FilterDataByType_5 InsideFilter AttributeMathsOp_7 InB)
  (connect CreateConstant_0 Out AttributeMathsOp_11 InA)
  (connect WriteDataIndex_1 Out AttributeMathsOp_13 InA)
  (connect WriteDataIndex_1 Out AttributeMathsOp_13 InB)
  (connect AttributeMathsOp_1 Out AttributeMathsOp_14 InA)
  (connect AttributeMathsOp_3 Out AttributeMathsOp_14 InB)
  (connect Branch_9 Output%20B AttributePartition_1 In)
  (connect MatchAndSetAttributes_0 Out AttributeRename_0 In)
  (connect MatchAndSetAttributes_1 Out AttributeRename_0 In)
  (connect FilterDataByTag_1 InsideFilter AttributeRename_1 In)
  (connect FilterDataByType_3 InsideFilter BoundsFromMesh_0 In)
  (connect Select_0 Out Branch_0 In)
  (connect MergePoints_1 Out Branch_1 In)
  (connect CopyAttributes_0 Out Branch_2 In)
  (connect AttributeCompareOp_1 Out Branch_3 bOutputToB)
  (connect Branch_8 Output%20B Branch_3 In)
  (connect CopyPoints_1 Out Branch_3 In)
  (connect DataNum_0 Out Branch_8 bOutputToB)
  (connect AttributeFilter_2 InsideFilter Branch_8 In)
  (connect DataNum_1 Out Branch_9 bOutputToB)
  (connect AttributeFilter_3 InsideFilter Branch_9 In)
  (connect DataNum_2 Out Branch_10 bOutputToB)
  (connect AttributeFilter_4 InsideFilter Branch_10 In)
  (connect AttributeMathsOp_9 Out BreakVectorAttribute_1 In)
  (connect MergePoints_0 Out CombinePoints_1 In)
  (connect AttributeMathsOp_6 Out CopyAttributes_0 Source)
  (connect AttributeMathsOp_7 Out CopyAttributes_0 Target)
  (connect AttributeRename_2 Out CopyAttributes_1 Source)
  (connect AttributeRename_2 Out CopyAttributes_1 Target)
  (connect AttributeRename_1 Out CopyPoints_1 Source)
  (connect AttributePartition_1 Out CopyPoints_1 Target)
  (connect AttributeFilter_2 InsideFilter DataNum_0 In)
  (connect AttributeFilter_3 InsideFilter DataNum_1 In)
  (connect AttributeFilter_4 InsideFilter DataNum_2 In)
  (connect Branch_2 Output%20B ExtractAttribute_1 In)
  (connect Branch_2 Output%20B ExtractAttribute_2 In)
  (connect Branch_2 Output%20B ExtractAttribute_3 In)
  (connect Branch_2 Output%20B ExtractAttribute_4 In)
  (connect Branch_2 Output%20B ExtractAttribute_5 In)
  (connect Branch_2 Output%20B ExtractAttribute_6 In)
  (connect AttributePartition_1 Out ExtractAttribute_7 In)
  (connect ToPoint_1 Out FilterDataByTag_1 In)
  (connect DefaultInputNode In FilterDataByType_0 In)
  (connect DefaultInputNode In FilterDataByType_1 In)
  (connect DefaultInputNode In FilterDataByType_2 In)
  (connect Branch_0 Output%20B FilterDataByType_3 In)
  (connect AttributeMathsOp_5 Out FilterDataByType_4 In)
  (connect Branch_0 Output%20A FilterDataByType_4 In)
  (connect Branch_0 Output%20B FilterDataByType_5 In)
  (connect StaticMeshSpawner_1 Out Gather_0 In)
  (connect StaticMeshSpawner_2 Out Gather_0 In2)
  (connect SpawnActor_1 Out Gather_0 In3)
  (connect AttributeFilter_0 OutsideFilter GetClassFromAttribute_0 In)
  (connect FilterDataByType_0 InsideFilter GetSplineControlPoints_0 In)
  (connect FilterDataByType_1 InsideFilter GetSplineControlPoints_0 In)
  (connect BreakVectorAttribute_0 Y MakeRotatorAttribute_0 Pitch)
  (connect BreakVectorAttribute_0 X MakeRotatorAttribute_0 Roll)
  (connect BreakVectorAttribute_0 Z MakeRotatorAttribute_0 Yaw)
  (connect BreakVectorAttribute_1 Y MakeRotatorAttribute_1 Pitch)
  (connect BreakVectorAttribute_1 X MakeRotatorAttribute_1 Roll)
  (connect BreakVectorAttribute_1 Z MakeRotatorAttribute_1 Yaw)
  (connect Branch_1 Output%20A MatchAndSetAttributes_0 In)
  (connect AttributeRename_2 Out MatchAndSetAttributes_0 Match%20Data)
  (connect Branch_1 Output%20B MatchAndSetAttributes_1 In)
  (connect CopyAttributes_1 Out MatchAndSetAttributes_1 Match%20Data)
  (connect GetSplineControlPoints_0 Out MergePoints_0 In)
  (connect AttributeMathsOp_13 Out MergePoints_1 In)
  (connect AttributeFilter_1 OutsideFilter PCGLoadDataAsset_1 In)
  (connect MatchAndSetAttributes_0 Out Select_0 Input%20A)
  (connect MatchAndSetAttributes_1 Out Select_0 Input%20A)
  (connect AttributeRename_0 Out Select_0 Input%20B)
  (connect Branch_10 Output%20B SpawnActor_1 In)
  (connect Branch_3 Output%20A StaticMeshSpawner_1 In)
  (connect AddAttribute_1 Out StaticMeshSpawner_2 In)
  (connect PCGLoadDataAsset_1 Out ToPoint_1 In)
  (connect FilterDataByType_4 InsideFilter TransformPoints_1 In)
  (connect TransformPoints_1 Out TransformPoints_2 In)
  (connect AttributeMathsOp_8 Out TransformPoints_2 OffsetMin)
  (connect MakeRotatorAttribute_1 Out TransformPoints_2 RotationMax)
  (connect MakeRotatorAttribute_0 Out TransformPoints_2 RotationMin)
  (connect AttributeMathsOp_10 Out TransformPoints_2 ScaleMax)
  (connect AttributeMathsOp_11 Out TransformPoints_2 ScaleMin)
  (connect CombinePoints_1 Out WriteDataIndex_1 In)
  (connect FilterDataByType_2 InsideFilter WriteDataIndex_1 In)
  (output Out Gather_0))
