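; PCG loop-body subgraph for asset spawning that expands input points into the point contents of PCG data assets.
; Input points become an attribute set that drives Load PCG Data Asset; loaded data converts to points and filters to the "points" tag.
; The Mesh attribute is renamed to Asset, then those points are copied onto the original input points as output.
; Position, rotation and scale offset parameters are declared but unused here.
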(graph Spawn_Assets_CoreProcessLoop
  (params (position_offset :type Vector) (Rotation_offset :type Rotator) (Scale_offset :type Vector :default (1.0 1.0 1.0)))
  (bind PCGLoadDataAsset_0 (Load%20PCG%20Data%20Asset))
  (bind ToPoint_0 (To%20Point))
  (bind FilterDataByTag_0 (Filter%20Data%20By%20Tag :SelectedTags "points"))
  (bind AttributeRename_0 (Attribute%20Rename :AttributeToRename Mesh :NewAttributeName Asset))
  (bind CopyPoints_0 (Copy%20Points))
  ;[0 0 0 0.5] Core Process
  (bind PointToAttributeSet_1 (Point%20To%20Attribute%20Set))
  (connect FilterDataByTag_0 InsideFilter AttributeRename_0 In)
  (connect AttributeRename_0 Out CopyPoints_0 Source)
  (connect DefaultInputNode In CopyPoints_0 Target)
  (connect ToPoint_0 Out FilterDataByTag_0 In)
  (connect PointToAttributeSet_1 Out PCGLoadDataAsset_0 Asset)
  (connect DefaultInputNode In PointToAttributeSet_1 In)
  (connect PCGLoadDataAsset_0 Out ToPoint_0 In)
  (output Out CopyPoints_0))
