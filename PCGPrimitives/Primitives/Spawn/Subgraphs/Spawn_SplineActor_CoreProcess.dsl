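; PCG subgraph that spawns an actor carrying spline components built from the incoming spline data.
; Input data is stripped of tags matching the substring "input", then filtered to spline data only.
; A plain Actor, created attached with its label and tags driven by graph parameters, serves as the target for Spawn Spline Component, whose result is the output.
; The Transform parameter is declared but not used.
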(graph Spawn_SplineActor_CoreProcess
  (params (Transform :type Transform) (Tags :type String) (Actor_label :type String :default "Actor"))
  (bind CreateTargetActor_0 (Create%20Target%20Actor :AttachOptions Attached :TemplateActorClass /Script/Engine.Actor :ActorLabel-param Actor_label :TagsToAddOnActor-param Tags))
  (bind DeleteTags_0 (Delete%20Tags :Operator Substring :SelectedTags "input"))
  (bind FilterDataByType_0 (Filter%20Data%20By%20Type :TargetType (struct :Ids [(struct :Struct /Script/PCG.PCGDataTypeInfoSpline)])))
  (bind SpawnSplineComponent_1 (Spawn%20Spline%20Component))
  (connect DefaultInputNode In DeleteTags_0 In)
  (connect DeleteTags_0 Out FilterDataByType_0 In)
  (connect FilterDataByType_0 InsideFilter SpawnSplineComponent_1 In)
  (connect CreateTargetActor_0 Out SpawnSplineComponent_1 TargetActor)
  (output Out SpawnSplineComponent_1))
